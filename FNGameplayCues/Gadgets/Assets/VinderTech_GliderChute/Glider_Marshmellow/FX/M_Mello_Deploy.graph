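# Translucent, unlit two-sided particle material for the Marshmellow glider's deploy effect.
# Color blends a radially panned rainbow gradient or gold-tinted digital noise via UseGoldBlack, tinted by particle color and HDR-boosted by the marshmallow mask.
# Opacity is a screen-space halftone dot pattern sized and shaped by that mask, eroded through dynamic parameters.
# The AttachtoCamera switch either pins the mesh to the camera via world position offset or applies near-camera fading.
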Material /FNGameplayCues/Gadgets/Assets/VinderTech_GliderChute/Glider_Marshmellow/FX/M_Mello_Deploy
graph {
 properties {
  bUsedWithNiagaraSprites=true
  bUsedWithParticleSprites=true
  blendMode=BLEND_Translucent
  pixelDepthOffsetMode=PDOM_Legacy
  shadingModel=MSM_Unlit
  twoSided=true
  usedShadingModels=Unlit
 }
 nodes [
  Add()
  n1 = Panner(
   Coordinate=MaterialFunctionCall() {materialFunction=&VectorToRadialValue}.Radial Coordinates
   Time=Multiply(A=Time(), B=DynamicParameter() {defaultValue=<rgba 1 1 0.25 0.25>, paramNames [PanningSpeed,Erode,ScaleU,ScaleV]}.PanningSpeed)
  ) {speedX=1.667}
  n2 = DynamicParameter() {defaultValue=<rgba 1 1 0.25 0.25>, paramNames [PanningSpeed,Erode,ScaleU,ScaleV]}
  n3 = Multiply(A=Time(), B=@n2.PanningSpeed)
  n4 = ComponentMask(None=VectorParameter() {parameterName=RandomSpeed, defaultValue=<rgba 1 1 0 0>}.RGB) {g=true, r=true}
  n5 = TextureSample(UVs=MaterialFunctionCall(UVs=Panner(Time=@n3, Speed=@n4) {speedX=1.667, speedY=0.2}, Texture Scale=AppendVector(A=@n2.ScaleU, B=@n2.ScaleU)) {materialFunction=&ScaleUVsByCenter}.UVs) {samplerType=SAMPLERTYPE_LinearColor, texture=/Game/Global/Textures/Default/Noise/T_Atlas_DigitalNoise_01}
  n6 = VectorParameter() {parameterName=Gold Color, defaultValue=<rgba 0.4444 0.3309 0 0>}
  n7 = Multiply(
   A=LinearInterpolate(
    A=TextureSample(UVs=@n1) {texture=/Game/Global/Textures/Default/Masks_and_Gradients/T_Grad_Rainbow}.RGB
    B=Multiply(A=@n5.R, B=@n6.RGB)
    Alpha=ScalarParameter() {parameterName=UseGoldBlack}
   )
   B=ParticleColor().RGB
  )
  n8 = TextureSample() {texture=&/T_Marshmellow_Mask}
  MaterialFunctionCall(Emissive=MaterialFunctionCall(LDR Input=Clamp(None=@n7), HDR Tint=@n7, HDR Intensity Multiplier=LinearInterpolate(Alpha=@n8.G) {constA=1, constB=10}) {materialFunction=&DeriveHDRfromLDR}.RemappedRange) {materialFunction=&MF_FX_EmissiveAdjustments}
  n9 = ScreenPosition()
  n10 = Multiply(A=@n8.RGB, B=ParticleColor().RGB)
  n11 = SphereMask(
   A=Frac(None=Divide(A=@n9.PixelPosition) {constB=5})
   B=Constant() {r=0.5}
   Radius=Add(A=ComponentMask(None=@n10) {r=true}, B=ComponentMask(None=@n10) {g=true})
   Hardness=Constant()
  )
  n12 = Multiply(
   A=Multiply(A=@n11, B=@n8.R)
   B=MaterialFunctionCall!(Dynamic Param=DynamicParameter() {defaultValue=<rgba 1 1 0.25 0.25>, paramNames [PanningSpeed,Erode,ScaleU,ScaleV]}.Erode, Alpha=@n11) {materialFunction=/Game/Effects/Fort_Effects/Materials/_Functions/Erode_Alpha}.Result
  )
  n13 = ScalarParameter() {defaultValue=250, parameterName=Near Camera Fade}
  StaticSwitchParameter(
   True=@n12
   False=Multiply(A=@n12, B=MaterialFunctionCall(Distance=@n13) {materialFunction=&Near_Camera_Fade}.Result)
  ) {parameterName=AttachtoCamera}
  StaticSwitchParameter(
   True=MaterialFunctionCall!(
    Object Basis 1
    Object Basis 2
    Object Basis 3
    Camera Space Position Offset=VectorParameter() {parameterName=CameraAttachposition, defaultValue=<rgba 0 0 257.8326 1>}.RGB
   ) {materialFunction=/Engine/Functions/Engine_MaterialFunctions02/WorldPositionOffset/AttachMeshToTheCamera}.World Position Offset
   False=Constant3Vector()
  ) {parameterName=AttachtoCamera}
 ]
}
