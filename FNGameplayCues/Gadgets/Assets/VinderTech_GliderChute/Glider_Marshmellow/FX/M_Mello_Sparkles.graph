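# Translucent unlit sparkle particle material for the Marshmallow glider effects, usable on Niagara and Cascade sprites.
# Particle random drives a three-way blend of sparkle, spark and starburst masks, which shapes emissive and, with particle alpha, opacity.
# Color cycles over time through a zero-tiled panning rainbow gradient, or UseGoldColor switches to repeatedly brightening gold, mixed with black by the Color Random dynamic parameter.
# Particle color tints the result before HDR remapping and emissive adjustment.
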Material /FNGameplayCues/Gadgets/Assets/VinderTech_GliderChute/Glider_Marshmellow/FX/M_Mello_Sparkles
graph {
 properties {
  bAllowNegativeEmissiveColor=true
  bUsedWithNiagaraSprites=true
  bUsedWithParticleSprites=true
  blendMode=BLEND_Translucent
  pixelDepthOffsetMode=PDOM_Legacy
  shadingModel=MSM_Unlit
  usedShadingModels=Unlit
 }
 nodes [
  n1 = ParticleRandom()
  n2 = MaterialFunctionCall(
   A=TextureSample() {texture=/Game/Effects/Fort_Effects/Textures/Sparkle/T_Sparkle_01}.R
   B=TextureSample() {texture=/Game/CharClasses/Outlander/Gadgets/AntiMaterialCharge/Materials/T_AntiMaterialCharge_Sparks_01}.G
   C=TextureSample() {texture=/Game/Effects/Fort_Effects/Textures/Flares/T_Starburst}.G
   Alpha=Ceil(None=Multiply(A=@n1) {constB=3})
  ) {materialFunction=&3ColorBlend}
  n3 = Panner(Coordinate=ComponentMask(None=TextureCoordinate() {uTiling=0, vTiling=0}) {g=true}) {speedX=1.6667}
  n4 = VectorParameter() {parameterName=Gold Color, defaultValue=<rgba 1 0.4167 0 1>}
  n5 = LinearInterpolate(Alpha=Time() {bOverride_Period=true, period=1.6667}) {constA=1, constB=2}
  n6 = ScalarParameter() {parameterName=UseGoldColor}
  n7 = Multiply(
   A=LinearInterpolate(
    A=TextureSample(UVs=@n3) {texture=/Game/Global/Textures/Default/Masks_and_Gradients/T_Grad_Rainbow}.RGB
    B=Multiply(A=@n4.RGB, B=@n5)
    Alpha=@n6
   )
   B=ParticleColor().RGB
  )
  n8 = LinearInterpolate(
   A=@n7
   B=LinearInterpolate(
    A=VectorParameter() {parameterName=Black Color}.RGB
    B=@n7
    Alpha=DynamicParameter() {paramNames [Color Random, Param2, Param3, Param4]}.Color Random
   )
   Alpha=@n6
  )
  n9 = Clamp(None=@n8)
  MaterialFunctionCall(Emissive=Multiply(
    A=@n2.Result
    B=MaterialFunctionCall(LDR Input=@n9, HDR Tint=@n8) {materialFunction=&DeriveHDRfromLDR}.RemappedRange
   )) {materialFunction=&MF_FX_EmissiveAdjustments}
  Multiply(A=@n2.Result, B=ParticleColor().A)
 ]
}
